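# Animation Blueprint for the YamPowder backpack cosmetic that adds physics-driven secondary motion to its three vial bones.
# Two chained sets of spring-based, zero-gravity Anim Dynamics nodes process the input pose, one enabled during gameplay and one in the front end.
# The front-end set has tighter angular limits and smaller physics boxes.
# Each update sets Not_in_FE to the inverse of bIsInFrontEnd, so only one set is active.
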AnimBlueprint /BRCosmetics/Accessories/FORT_Backpacks/Backpack_M_MED_YamPowder/Meshes/M_MED_YamPowder_Pack_AnimBP
target_skeleton=&/M_MED_YamPowder_Pack_Skeleton
graphs [
 AnimGraph {nodes [
   n1 = Get Not_in_FE() -> (Not_in_FE:bool)
   n2 = Anim Dynamics - Bone: vial_01!(ComponentPose=Local To Component(LocalPose=Input Pose()), bAlphaBoolEnabled:bool=@n1, ExternalForce="0,0,0") -> (Pose) {
    alphaInputType=Bool
    angularDampingOverride=1
    angularSpringConstant=500
    bAngularSpring=true
    bOverrideAngularDamping=true
    gravityScale=0
    numSolverIterationsPreUpdate=3
    alphaBoolBlend {blendOption=Linear}
    alphaScaleBiasClamp {bClampResult=true}
    boundBone {boneName=vial_01}
    componentAppliedLinearAccClamp=<400 400 400>
    componentLinearAccScale=<0.4 0.4 0.4>
    componentLinearVelScale=<0.5 0.5 0.5>
    physicsBodyDefinitions [{
      boundBone {boneName=vial_01}
      boxExtents=<8 8 8>
      constraintSetup {angularLimitsMax=<0 10 0>, angularLimitsMin=<0 -10 0>, angularTarget=<1 0 0>}
      localJointOffset=<0.7 0 0>
     }]
   }
   n3 = Anim Dynamics - Bone: vial_02!(ComponentPose=@n2, bAlphaBoolEnabled:bool=@n1, ExternalForce="0,0,0") -> (Pose) {
    alphaInputType=Bool
    angularDampingOverride=0.9999
    angularSpringConstant=3000
    bAngularSpring=true
    bOverrideAngularDamping=true
    gravityScale=0
    numSolverIterationsPreUpdate=3
    alphaBoolBlend {blendOption=Linear}
    alphaScaleBiasClamp {bClampResult=true}
    boundBone {boneName=vial_02}
    componentAppliedLinearAccClamp=<300 500 500>
    componentLinearAccScale=<0.2 0.2 0.3>
    componentLinearVelScale=<0.3 0.3 0.2>
    physicsBodyDefinitions [{
      boundBone {boneName=vial_02}
      boxExtents=<7 12 12>
      constraintSetup {angularLimitsMax=<0 10 0>, angularLimitsMin=<0 -10 0>, angularTarget=<1 0 0>}
      localJointOffset=<2 0 0>
     }]
   }
   n4 = Anim Dynamics - Bone: vial_03!(ComponentPose=@n3, bAlphaBoolEnabled:bool=@n1, ExternalForce="0,0,0") -> (Pose) {
    alphaInputType=Bool
    angularDampingOverride=1
    angularSpringConstant=2000
    bAngularSpring=true
    bOverrideAngularDamping=true
    gravityScale=0
    alphaBoolBlend {blendOption=Linear}
    alphaScaleBiasClamp {bClampResult=true}
    boundBone {boneName=vial_03}
    componentAppliedLinearAccClamp=<300 500 500>
    componentLinearAccScale=<0.4 0.4 0.4>
    componentLinearVelScale=<0.5 0.5 0.5>
    physicsBodyDefinitions [{
      boundBone {boneName=vial_03}
      boxExtents=<10 10 10>
      constraintSetup {angularLimitsMax=<0 10 0>, angularLimitsMin=<0 -10 0>, angularTarget=<1 0 0>}
      localJointOffset=<1 0 0>
     }]
   }
   n5 = Get bIsInFrontEnd()
   n6 = Anim Dynamics - Bone: vial_01(ComponentPose=@n4, bAlphaBoolEnabled=@n5, ExternalForce="0,0,0") {
    alphaInputType=Bool
    angularDampingOverride=1
    angularSpringConstant=500
    bAngularSpring=true
    bOverrideAngularDamping=true
    gravityScale=0
    numSolverIterationsPreUpdate=3
    alphaBoolBlend {blendOption=Linear}
    alphaScaleBiasClamp {bClampResult=true}
    boundBone {boneName=vial_01}
    componentAppliedLinearAccClamp=<400 400 400>
    componentLinearAccScale=<0.4 0.4 0.4>
    componentLinearVelScale=<0.5 0.5 0.5>
    physicsBodyDefinitions [{
      boundBone {boneName=vial_01}
      boxExtents=<7 7 7>
      constraintSetup {angularLimitsMax=<0 3.5 0>, angularLimitsMin=<0 -3.5 0>, angularTarget=<1 0 0>}
      localJointOffset=<0.7 0 0>
     }]
   }
   n7 = Anim Dynamics - Bone: vial_02(ComponentPose=@n6, bAlphaBoolEnabled=@n5, ExternalForce="0,0,0") {
    alphaInputType=Bool
    angularDampingOverride=0.9999
    angularSpringConstant=3000
    bAngularSpring=true
    bOverrideAngularDamping=true
    gravityScale=0
    numSolverIterationsPreUpdate=3
    alphaBoolBlend {blendOption=Linear}
    alphaScaleBiasClamp {bClampResult=true}
    boundBone {boneName=vial_02}
    componentAppliedLinearAccClamp=<300 500 500>
    componentLinearAccScale=<0.2 0.2 0.3>
    componentLinearVelScale=<0.3 0.3 0.2>
    physicsBodyDefinitions [{
      boundBone {boneName=vial_02}
      boxExtents=<6 10 10>
      constraintSetup {angularLimitsMax=<0 3.5 0>, angularLimitsMin=<0 -3.5 0>, angularTarget=<1 0 0>}
      localJointOffset=<2 0 0>
     }]
   }
   n8 = Anim Dynamics - Bone: vial_03(ComponentPose=@n7, bAlphaBoolEnabled=@n5, ExternalForce="0,0,0") {
    alphaInputType=Bool
    angularDampingOverride=1
    angularSpringConstant=2000
    bAngularSpring=true
    bOverrideAngularDamping=true
    gravityScale=0
    alphaBoolBlend {blendOption=Linear}
    alphaScaleBiasClamp {bClampResult=true}
    boundBone {boneName=vial_03}
    componentAppliedLinearAccClamp=<300 500 500>
    componentLinearAccScale=<0.4 0.4 0.4>
    componentLinearVelScale=<0.5 0.5 0.5>
    physicsBodyDefinitions [{
      boundBone {boneName=vial_03}
      boxExtents=<9 9 9>
      constraintSetup {angularLimitsMax=<0 3.5 0>, angularLimitsMin=<0 -3.5 0>, angularTarget=<1 0 0>}
      localJointOffset=<1 0 0>
     }]
   }
   Output Pose(Result=Component To Local(ComponentPose=@n8)) {layerGroup=None}
  ]}
 EventGraph {nodes [
   Event BlueprintUpdateAnimation()
   > Set Not_in_FE(Not_in_FE:bool=NOT Boolean(A=Get bIsInFrontEnd())) -> (Output_Get:bool)
   TryGetPawnOwner()
  ]}
]
variables [Not_in_FE {type=boolean}]
